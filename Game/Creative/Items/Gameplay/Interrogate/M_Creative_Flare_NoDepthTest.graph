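# Unlit additive flare material for particle and Niagara sprites, drawn without depth testing so it shows through scene geometry, with responsive AA enabled.
# A grayscale flare texture parameter, defaulting to a cross flare, is raised to an Exponent parameter to shape its falloff and multiplied by particle color.
# The result feeds the MF_FX_EmissiveAdjustments material function.
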Material /Game/Creative/Items/Gameplay/Interrogate/M_Creative_Flare_NoDepthTest
graph {
 properties {
  bDisableDepthTest=true
  bEnableResponsiveAA=true
  bUsedWithNiagaraSprites=true
  bUsedWithParticleSprites=true
  blendMode=BLEND_Additive
  pixelDepthOffsetMode=PDOM_Legacy
  shadingModel=MSM_Unlit
  usedShadingModels=Unlit
 }
 nodes [
  n1 = TextureSampleParameter2D() {parameterName=Texture, samplerType=SAMPLERTYPE_Grayscale, texture=/Game/Effects/Fort_Effects/Textures/Flares/T_Cross_Flare}
  n2 = ScalarParameter() {defaultValue=2, parameterName=Exponent}
  MaterialFunctionCall(Emissive=Multiply(A=ParticleColor().RGB, B=Power(Base=@n1.RGB, Exp=@n2))) {materialFunction=&MF_FX_EmissiveAdjustments}
 ]
}
